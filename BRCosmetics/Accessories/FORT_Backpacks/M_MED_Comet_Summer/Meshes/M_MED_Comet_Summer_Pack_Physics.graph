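# Physics asset for the M_MED_Comet_Summer backpack cosmetic, rigging its dangling character parts (arms, legs, ears, hair and hat strings) for physics motion.
# Each side's moving bones use capsules constrained to small-sphere base bodies, with hat strings forming three-capsule chains.
# Swing limits vary by part, tightest on the hair, and twist stays narrow throughout.
# Base bodies and the backpack offset sphere have no constraints linking them together.
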PhysicsAsset /BRCosmetics/Accessories/FORT_Backpacks/M_MED_Comet_Summer/Meshes/M_MED_Comet_Summer_Pack_Physics
bodies {
 dyn_arm_base_l [{shapeType=Sphere, radius=0.5}]
 dyn_arm_base_r [{shapeType=Sphere, radius=0.5}]
 dyn_arm_l [{shapeType=Capsule, length=1.875, radius=1.38, center=<1.7993 0 0>, rotation=<rot 90 0 0>}]
 dyn_arm_r [{shapeType=Capsule, length=1.6694, radius=1.4683, center=<1.6928 0 0>, rotation=<rot 90 0 0>}]
 dyn_backpack_offset [{shapeType=Sphere, radius=0.5}]
 dyn_ear_base_l [{shapeType=Sphere, radius=0.5}]
 dyn_ear_base_r [{shapeType=Sphere, radius=0.5}]
 dyn_ear_l [{shapeType=Capsule, length=1.375, radius=1.13, center=<1.2062 0 0>, rotation=<rot 90 0 0>}]
 dyn_ear_r [{shapeType=Capsule, length=1.236, radius=0.985, center=<1.054 0 0>, rotation=<rot 90 0 0>}]
 dyn_hair_base_l [{shapeType=Sphere, radius=0.5}]
 dyn_hair_base_r [{shapeType=Sphere, radius=0.5}]
 dyn_hair_l [{shapeType=Capsule, length=1.255, radius=1.38, center=<1.4149 0 0>, rotation=<rot 90 0 0>}]
 dyn_hair_r [{shapeType=Capsule, length=1.5373, radius=1.3682, center=<1.5618 0 0>, rotation=<rot 90 0 0>}]
 dyn_hat_string_01_l [{shapeType=Capsule, length=0.8862, radius=0.748, center=<0.937 0 0>, rotation=<rot 90 0 0>}]
 dyn_hat_string_01_r [{shapeType=Capsule, length=0.7612, radius=0.7448, center=<1.0203 0 0>, rotation=<rot 90 0 0>}]
 dyn_hat_string_02_l [{shapeType=Capsule, length=0.8525, radius=0.8001, center=<0.8956 0 0>, rotation=<rot 90 0 0>}]
 dyn_hat_string_02_r [{shapeType=Capsule, length=0.9775, radius=0.6725, center=<0.9655 0 0>, rotation=<rot 90 0 0>}]
 dyn_hat_string_base_l [{shapeType=Sphere, radius=0.5}]
 dyn_hat_string_base_r [{shapeType=Sphere, radius=0.5}]
 dyn_hat_string_l [{shapeType=Capsule, length=0.755, radius=0.755, center=<1.0744 0 0>, rotation=<rot 90 0 0>}]
 dyn_hat_string_r [{shapeType=Capsule, length=0.9686, radius=0.6963, center=<1.0824 0 0>, rotation=<rot 90 0 0>}]
 dyn_leg_base_l [{shapeType=Sphere, radius=0.5}]
 dyn_leg_base_r [{shapeType=Sphere, radius=0.5}]
 dyn_leg_l [{shapeType=Capsule, length=1.375, radius=1.505, center=<2.028 0 0>, rotation=<rot 90 0 0>}]
 dyn_leg_r [{shapeType=Capsule, length=1.6323, radius=1.5497, center=<2.0003 0 0>, rotation=<rot 90 0 0>}]
}
constraints [
 {bone1Name=dyn_leg_l, bone2Name=dyn_leg_base_l, swing1LimitDegrees=15, swing2LimitDegrees=10, twistLimitDegrees=3}
 {bone1Name=dyn_arm_l, bone2Name=dyn_arm_base_l, swing1LimitDegrees=25, swing2LimitDegrees=20, twistLimitDegrees=3}
 {bone1Name=dyn_hair_l, bone2Name=dyn_hair_base_l, swing1LimitDegrees=4, swing2LimitDegrees=4, twistLimitDegrees=3}
 {bone1Name=dyn_ear_l, bone2Name=dyn_ear_base_l, swing1LimitDegrees=10, swing2LimitDegrees=10, twistLimitDegrees=3}
 {bone1Name=dyn_leg_r, bone2Name=dyn_leg_base_r, swing1LimitDegrees=20, swing2LimitDegrees=15, twistLimitDegrees=3}
 {bone1Name=dyn_arm_r, bone2Name=dyn_arm_base_r, swing1LimitDegrees=25, swing2LimitDegrees=20, twistLimitDegrees=3}
 {bone1Name=dyn_hair_r, bone2Name=dyn_hair_base_r, swing1LimitDegrees=4, swing2LimitDegrees=4, twistLimitDegrees=3}
 {bone1Name=dyn_ear_r, bone2Name=dyn_ear_base_r, swing1LimitDegrees=10, swing2LimitDegrees=10, twistLimitDegrees=3}
 {bone1Name=dyn_hat_string_l, bone2Name=dyn_hat_string_base_l, swing1LimitDegrees=15, swing2LimitDegrees=10, twistLimitDegrees=3}
 {bone1Name=dyn_hat_string_01_l, bone2Name=dyn_hat_string_l, swing1LimitDegrees=15, swing2LimitDegrees=10, twistLimitDegrees=3}
 {bone1Name=dyn_hat_string_02_l, bone2Name=dyn_hat_string_01_l, swing1LimitDegrees=15, swing2LimitDegrees=10, twistLimitDegrees=3}
 {bone1Name=dyn_hat_string_02_r, bone2Name=dyn_hat_string_01_r, swing1LimitDegrees=15, swing2LimitDegrees=10, twistLimitDegrees=3}
 {bone1Name=dyn_hat_string_01_r, bone2Name=dyn_hat_string_r, swing1LimitDegrees=15, swing2LimitDegrees=10, twistLimitDegrees=3}
 {bone1Name=dyn_hat_string_r, bone2Name=dyn_hat_string_base_r, swing1LimitDegrees=15, swing2LimitDegrees=10, twistLimitDegrees=3}
]
