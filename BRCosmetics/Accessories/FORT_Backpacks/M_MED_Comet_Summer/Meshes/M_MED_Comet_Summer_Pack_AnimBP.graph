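# Animation Blueprint for the Comet Summer backpack cosmetic that adds physics-driven secondary motion to the pack.
# A RigidBody node simulates the pack's physics asset in root base-bone space over the input pose, with overridden gravity and scaled, clamped component motion.
# Jump and fall events set an upward Jump_Force fed in as external force, clearing it when jumping ends or half a second after falling ends.
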AnimBlueprint /BRCosmetics/Accessories/FORT_Backpacks/M_MED_Comet_Summer/Meshes/M_MED_Comet_Summer_Pack_AnimBP
target_skeleton=&/M_MED_Comet_Summer_Pack_Skeleton
graphs [
 AnimGraph {nodes [
   n1 = RigidBody(ComponentPose=Local To Component(LocalPose=Input Pose()), ExternalForce=Get Jump_Force() -> (Jump_Force:Vector)) {
    bClampLinearTranslationLimitToRefPose=true
    bOverrideWorldGravity=true
    overridePhysicsAsset=&/M_MED_Comet_Summer_Pack_Physics
    simulationSpace=BaseBoneSpace
    alphaBoolBlend {blendOption=Linear}
    alphaScaleBiasClamp {bias=0.1, scale=0.8}
    baseBoneRef {boneName=root}
    componentAppliedLinearAccClamp=<3000 3000 3000>
    componentLinearAccScale=<0.25 0.25 0.25>
    componentLinearVelScale=<0.25 0.25 0.25>
    overrideWorldGravity {z=-920}
   }
   Output Pose(Result=Component To Local(ComponentPose=@n1)) {layerGroup=None}
  ]}
 EventGraph {nodes [
   Event OnBeginJumping()
   > Set Jump_Force(Jump_Force:Vector="0, 0,700") -> (Output_Get:Vector)
   Event OnEndJumping()
   > Set Jump_Force(Jump_Force="0, 0, 0")
   Event OnBeginFalling()
   > Set Jump_Force(Jump_Force="0, 0,700")
   Event OnEndFalling()
   > Delay(Duration=0.5)
   > Set Jump_Force(Jump_Force="0, 0, 0")
   TryGetPawnOwner()
   Event BlueprintUpdateAnimation()
  ]}
]
variables [Jump_Force {type=Vector}]
